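# Sound cue for the impact sound of the Grappler bow.
# A random node picks one of four impact wave variations with equal weight and feeds a modulator that varies the volume slightly above unity on each play.
# It plays in the impacts sound class with gadget-activate attenuation and borrows the rifle bullet impact concurrency to manage overlapping hits.
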SoundCue /PrimalGameplay/Sounds/Weapons/Bows/Grappler/Bow_Grappler_Impact_Cue
root=Modulator_0
nodes [
 {id=Modulator_0, children [Random_2], properties {volumeMax=1.2, volumeMin=1.1}}
 {id=Random_2, children [WavePlayer_0,WavePlayer_1,WavePlayer_2,WavePlayer_3], properties {weights [1,1,1,1]}}
 {id=WavePlayer_0, properties {soundWaveAssetPtr=&/Bow_Grappler_Impact_04}}
 {id=WavePlayer_1, properties {soundWaveAssetPtr=&/Bow_Grappler_Impact_01}}
 {id=WavePlayer_2, properties {soundWaveAssetPtr=&/Bow_Grappler_Impact_02}}
 {id=WavePlayer_3, properties {soundWaveAssetPtr=&/Bow_Grappler_Impact_03}}
]
properties {
 attenuationSettings=&Fort_GadgetActivate_Att
 maxDistance=4000
 soundClassObject=&Fort_Impacts
 volumeMultiplier=1
 attenuationOverrides {bEnableReverbSend=false, falloffDistance=3000, nonSpatializedRadiusStart=1000, stereoSpread=2000, attenuationShapeExtents {x=2500}}
 concurrencySet [&RifleBulletImpact_Conc]
}
