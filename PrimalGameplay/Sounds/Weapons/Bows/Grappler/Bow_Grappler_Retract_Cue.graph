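# Sound cue for the retract sound of the grappler bow weapon.
# A random node picks one of four equally weighted retract waves, and a modulator varies its pitch slightly on each play.
# It plays at reduced volume in the gadgets sound class, using gadget grappling attenuation and a concurrency rule named for two instances per owner, stopping the oldest.
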SoundCue /PrimalGameplay/Sounds/Weapons/Bows/Grappler/Bow_Grappler_Retract_Cue
root=Modulator_1
nodes [
 {id=Modulator_1, children [Random_0], properties {pitchMax=1.1, pitchMin=0.9}}
 {id=Random_0, children [WavePlayer_4,WavePlayer_5,WavePlayer_6,WavePlayer_7], properties {weights [1,1,1,1]}}
 {id=WavePlayer_4, properties {soundWaveAssetPtr=&/Bow_Grappler_Retract_01}}
 {id=WavePlayer_5, properties {soundWaveAssetPtr=&/Bow_Grappler_Retract_02}}
 {id=WavePlayer_6, properties {soundWaveAssetPtr=&/Bow_Grappler_Retract_03}}
 {id=WavePlayer_7, properties {soundWaveAssetPtr=&/Bow_Grappler_Retract_04}}
]
properties {
 attenuationSettings=/Game/Sounds/Attenuation/Fort_GadgetGrappling_Att
 maxDistance=8500
 soundClassObject=&Fort_Gadgets
 volumeMultiplier=0.6
 concurrencySet [/Game/Sounds/Concurrency_Rules/TwoMaximum_Owner_StopOldest]
}
