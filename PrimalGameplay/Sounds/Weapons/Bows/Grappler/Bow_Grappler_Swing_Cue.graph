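# Sound cue for the grappler bow's swing, played as a spatialized gadget sound.
# A mixer layers one random pick from four swing waves with one from four retract waves, all equally weighted, then feeds a modulator at the root.
# It overrides attenuation with a custom falloff, non-spatialized near radius and wide stereo spread, with reverb send off, and limits voices through the player melee weapon concurrency set.
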SoundCue /PrimalGameplay/Sounds/Weapons/Bows/Grappler/Bow_Grappler_Swing_Cue
root=Modulator_2
nodes [
 {id=Mixer_0, children [Random_0,Random_1], properties {inputVolume [1,1]}}
 {id=Modulator_2, children [Mixer_0]}
 {id=Random_0, children [WavePlayer_0,WavePlayer_1,WavePlayer_2,WavePlayer_3], properties {weights [1,1,1,1]}}
 {id=Random_1, children [WavePlayer_8, WavePlayer_9, WavePlayer_10, WavePlayer_11], properties {weights [1,1,1,1]}}
 {id=WavePlayer_0, properties {soundWaveAssetPtr=&/Bow_Grappler_Swing_01}}
 {id=WavePlayer_1, properties {soundWaveAssetPtr=&/Bow_Grappler_Swing_02}}
 {id=WavePlayer_10, properties {soundWaveAssetPtr=&/Bow_Grappler_Retract_03}}
 {id=WavePlayer_11, properties {soundWaveAssetPtr=&/Bow_Grappler_Retract_04}}
 {id=WavePlayer_2, properties {soundWaveAssetPtr=&/Bow_Grappler_Swing_03}}
 {id=WavePlayer_3, properties {soundWaveAssetPtr=&/Bow_Grappler_Swing_04}}
 {id=WavePlayer_8, properties {soundWaveAssetPtr=&/Bow_Grappler_Retract_01}}
 {id=WavePlayer_9, properties {soundWaveAssetPtr=&/Bow_Grappler_Retract_02}}
]
properties {
 attenuationSettings=/Game/Sounds/Attenuation/Fort_GadgetGrappling_Att
 bOverrideAttenuation=true
 maxDistance=3000
 soundClassObject=&Fort_Gadgets
 attenuationOverrides {bEnableReverbSend=false, falloffDistance=3000, nonSpatializedRadiusStart=800, stereoSpread=2000, attenuationShapeExtents {x=0}}
 concurrencySet [&Player_Melee_Weapon_Conc]
}
